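# Grass material function that outputs a biome-tinted base color and a world-noise overlay.
# Biome masks blend four grass colors, and an optional static switch darkens them with fake camera-facing grass cards painted in the shader.
# World-XY cells use vector noise to pick atlas flipbook frames and thin out cards, which appear only with distance from the camera.
# The slope-weighted and quantized noise chains reach no output.
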MaterialFunction /WildEstateTech/Environment/BaseMaterials/MaterialFunctions/MF_WildEstate_Grass
graph {nodes [
  n1 = FunctionInput(Preview=VertexNormalWS()) {bUsePreviewValueAsDefault=true, inputName=NormalOverrideWS}
  n2 = Multiply(A=Abs(None=ComponentMask(None=@n1) {r=true}), B=ScalarParameter() {defaultValue=-0.7611, parameterName=Grass_Quanitize_SlopeInfluence_X})
  n3 = Multiply(A=Abs(None=ComponentMask(None=@n1) {g=true}), B=ScalarParameter() {defaultValue=0.2709, parameterName=Grass_Quanitize_SlopeInfluence_Y})
  n4 = ComponentMask(None=@n1) {b=true}
  n5 = ScalarParameter() {defaultValue=0.4752, parameterName=Grass_Quanitize_SlopeInfluence}
  n6 = TextureSample(UVs=Divide(A=WorldPosition().XY, B=ScalarParameter() {defaultValue=22012.3652, parameterName=Grass_Noise_Size})) {samplerType=SAMPLERTYPE_LinearGrayscale, texture=/WildEstateTech/Environment/TestContent/T_WildEstate_Landscape_Noise_M}
  Add(A=Add(A=Add(A=@n2, B=@n3), B=Multiply(A=@n4, B=@n5)), B=@n6.R)
  Add()
  n7 = Divide(A=AppendVector(A=WorldPosition().XY, B=Constant()), B=ScalarParameter() {defaultValue=720.6542, parameterName=Grass_Card_CellSize})
  n8 = Subtract(A=Frac(None=@n7)) {constB=0.5}
  AppendVector(A=@n8, B=Constant())
  Constant() {r=0.5}
  Floor(None=@n7)
  Frac(None=ComponentMask(None=@n7) {g=true, r=true})
  FunctionOutput(None=NamedRerouteDeclaration(None=MaterialFunctionCall!(NormalOverrideWS) {materialFunction=&/MF_WildEstateWorldNoise}.Result) {name=Grass Noise, nodeColor=<rgba 1 0.6797 0 1>}) {outputName=Noise Overlay, sortPriority=50}
  n9 = VectorParameter() {parameterName=GrassColor_Default, defaultValue=<rgba 0.0252 0.1946 0.0497 1>}
  n10 = VectorParameter() {parameterName=GrassColor_BiomeA, defaultValue=<rgba 0.0169 0.1146 0.052 1>}
  n11 = MaterialFunctionCall!(
   UseLandscapeCoords=FunctionInput(Preview=StaticBool() {value=true}) {bUsePreviewValueAsDefault=true, inputName=UseLandscapeCoords, inputType=FunctionInput_StaticBool, sortPriority=10}
   "UseProjection(translucent)"
  ) {materialFunction=/WildEstateTech/Landscape/MaterialFunctions/MF_WildEstate_BiomeMasks}
  n12 = MaterialFunctionCall(Float3=@n11.QuadA_RGBA) {materialFunction=&BreakOutFloat3Components}
  n13 = LinearInterpolate(
   A=LinearInterpolate(
    A=LinearInterpolate(A=@n9.RGB, B=@n10.RGB, Alpha=@n12.R)
    B=VectorParameter() {parameterName=GrassColor_BiomeC, defaultValue=<rgba 0.1118 0.2344 0.0269 1>}.RGB
    Alpha=@n12.G
   )
   B=VectorParameter() {parameterName=GrassColor_BiomeD, defaultValue=<rgba 0.2344 0.1747 0.0818 1>}.RGB
   Alpha=MaterialFunctionCall(Float3=@n11.QuadB_RGBA) {materialFunction=&BreakOutFloat3Components}.R
  )
  n14 = VectorNoise(World Position=@n7)
  n15 = ComponentMask(None=@n14) {g=true}
  n16 = Constant() {r=4}
  n17 = TextureObject() {texture=/WildEstateTech/Environment/TestContent/WildEstate_GrassCard_Atlas}
  n18 = Multiply(
   A=Transform(None=Constant3Vector() {constant=<rgba 1 0 0 0>}) {transformSourceType=TRANSFORMSOURCE_Camera, transformType=TRANSFORM_Tangent}
   B=Constant3Vector() {constant=<rgba 1 1 0 0>}
  )
  n19 = Constant3Vector() {constant=<rgba 0 0 1 0>}
  n20 = Normalize(VectorInput=Multiply(
    A=Transform(None=@n19) {transformSourceType=TRANSFORMSOURCE_Camera, transformType=TRANSFORM_Tangent}
    B=Constant3Vector() {constant=<rgba 1 1 0 0>}
   ))
  n21 = AppendVector(A=DotProduct(A=@n8, B=Normalize(VectorInput=@n18)), B=Multiply(A=DotProduct(A=@n8, B=@n20)) {constB=-1})
  n22 = Saturate(None=ComponentMask(None=Add(A=@n21) {constB=0.5}) {g=true, r=true})
  n23 = WorldPosition() {worldPositionShaderOffset=WPT_CameraRelativeNoOffsets}
  n24 = Divide(
   A=Subtract(A=Length(None=@n23.XYZ) {is=class MaterialExpressionLength}, B=ScalarParameter() {defaultValue=8682.4102, parameterName=FadeBias})
   B=ScalarParameter() {defaultValue=2000, parameterName=FadeRange}
  )
  n25 = Add(
   A=ComponentMask(None=MaterialFunctionCall("Animation  Phase (0-1)"=@n15, Number of Rows=@n16, Number of Columns=@n16, Texture=@n17, UVs=@n22) {materialFunction=&FlipBook}.Result) {r=true}
   B=OneMinus(None=Saturate(None=@n24))
  )
  n26 = Subtract(A=ComponentMask(None=@n14) {r=true}, B=ScalarParameter() {defaultValue=0.8157, parameterName=Threshold})
  n27 = Multiply(A=@n13, B=LinearInterpolate(B=Saturate(None=@n25), Alpha=Ceil(None=@n26)) {constA=1})
  n28 = FunctionInput(Preview=StaticBool() {value=true}) {bUsePreviewValueAsDefault=true, inputName=UseCamera facing BildBoards, inputType=FunctionInput_StaticBool}
  FunctionOutput(None=Multiply(A=StaticSwitch(True=@n27, False=@n13, Value=@n28), B=NamedRerouteUsage())) {outputName=Grass BaseColor, sortPriority=10}
  PixelNormalWS()
  n29 = ScalarParameter() {defaultValue=5, parameterName=GrassNoise_Quanitize_Steps}
  n30 = Divide(A=Round(None=Multiply(A=@n6.R, B=@n29)), B=@n29)
  Saturate(None=LinearInterpolate(
    A=ScalarParameter() {defaultValue=0.584, group=QuantizeOverlay, parameterName=Grass_Noise_Quantize_MinDarkness}
    B=ScalarParameter() {defaultValue=1, group=QuantizeOverlay, parameterName=Grass_Noise_Quantize_MaxBrigthness}
    Alpha=Saturate(None=@n30)
   ))
  ScalarParameter() {defaultValue=200, parameterName=DepthRipples}
 ]}
