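# Material function that produces a two-tone snow color, stored with the WildEstate landscape assets.
# It blends the SnowColor (pale) and SnowColor_2 (deeper blue) parameters, using the Noise Overlay output of MF_WildEstate_Grass as the mask.
# An optional NormalOverrideWS input, defaulting to the world-space vertex normal, passes into that grass function.
# The blended color is the only output, named Lerp.
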MaterialFunction /WildEstateTech/Landscape/M_WildEstate_Snow
graph {nodes [
  n1 = FunctionInput(Preview=VertexNormalWS()) {bUsePreviewValueAsDefault=true, inputName=NormalOverrideWS}
  FunctionOutput(None=LinearInterpolate(
    A=VectorParameter() {parameterName=SnowColor, defaultValue=<rgba 0.8 0.8139 0.8958 1>}.RGB
    B=VectorParameter() {parameterName=SnowColor_2, defaultValue=<rgba 0.2274 0.4397 0.6823 1>}.RGB
    Alpha=MaterialFunctionCall!(NormalOverrideWS=@n1, UseCamera facing BildBoards, UseLandscapeCoords) {materialFunction=/WildEstateTech/Environment/BaseMaterials/MaterialFunctions/MF_WildEstate_Grass}.Noise Overlay
   ) {constA=1}) {outputName=Lerp}
 ]}
